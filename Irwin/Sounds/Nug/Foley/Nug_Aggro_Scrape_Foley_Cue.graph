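# Sound cue for the Nug creature's aggro scrape foley, randomly picking one recorded variation per play.
# A quality-level node sends the first two levels to an eight-wave random pool and the last two to a pool of only the first four waves.
# It uses prey step attenuation and the tactical creature steps sound class, but shares the Nug global VO concurrency setting.
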SoundCue /Irwin/Sounds/Nug/Foley/Nug_Aggro_Scrape_Foley_Cue
root=QualityLevel_0
nodes [
 {id=QualityLevel_0, children [Random_0,Random_0,Random_1,Random_1]}
 {
  id=Random_0
  children [WavePlayer_0,WavePlayer_1,WavePlayer_2,WavePlayer_3,WavePlayer_4,WavePlayer_5,WavePlayer_6,WavePlayer_7]
  properties {weights [1,1,1,1,1,1,1,1]}
 }
 {
  id=Random_1
  children [WavePlayer_8, WavePlayer_9, WavePlayer_10, WavePlayer_11]
  properties {bShouldExcludeFromBranchCulling=true, weights [1,1,1,1]}
 }
 {id=WavePlayer_0, properties {soundWaveAssetPtr=&/Nug_Aggro_Scrape_Foley_01}}
 {id=WavePlayer_1, properties {soundWaveAssetPtr=&/Nug_Aggro_Scrape_Foley_02}}
 {id=WavePlayer_10, properties {soundWaveAssetPtr=&/Nug_Aggro_Scrape_Foley_03}}
 {id=WavePlayer_11, properties {soundWaveAssetPtr=&/Nug_Aggro_Scrape_Foley_04}}
 {id=WavePlayer_2, properties {soundWaveAssetPtr=&Nug_Aggro_Scrape_Foley_03}}
 {id=WavePlayer_3, properties {soundWaveAssetPtr=&Nug_Aggro_Scrape_Foley_04}}
 {id=WavePlayer_4, properties {soundWaveAssetPtr=&/Nug_Aggro_Scrape_Foley_05}}
 {id=WavePlayer_5, properties {soundWaveAssetPtr=&/Nug_Aggro_Scrape_Foley_06}}
 {id=WavePlayer_6, properties {soundWaveAssetPtr=&/Nug_Aggro_Scrape_Foley_07}}
 {id=WavePlayer_7, properties {soundWaveAssetPtr=&/Nug_Aggro_Scrape_Foley_08}}
 {id=WavePlayer_8, properties {soundWaveAssetPtr=&Nug_Aggro_Scrape_Foley_01}}
 {id=WavePlayer_9, properties {soundWaveAssetPtr=&Nug_Aggro_Scrape_Foley_02}}
]
properties {
 attenuationSettings=/Irwin/Sounds/Settings/Prey_Step_Att
 maxDistance=3200
 priority=7
 soundClassObject=/Game/Sounds/FortSoundClassesAndModes/Fort_Creatures_Steps_Tactical
 volumeMultiplier=1.5
 concurrencySet [/Irwin/Sounds/Settings/Conc/Nug_Global_VO_Conc]
}
