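# Sound cue for the Nug creature's flap foley, using the creature-steps sound class and prey step attenuation.
# A quality-level switch sends the first two levels to a random pick of seven flap waves and the last two to a pick of four, the latter excluded from branch culling.
# A modulator above the switch adds slight pitch variation, and a concurrency override caps simultaneous instances.
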SoundCue /Irwin/Sounds/Nug/Foley/Nug_Flap_Cue
root=Modulator_0
nodes [
 {id=Modulator_0, children [QualityLevel_0], properties {pitchMax=1.1, pitchMin=0.9}}
 {id=QualityLevel_0, children [Random_0,Random_0,Random_2,Random_2]}
 {
  id=Random_0
  children [WavePlayer_1,WavePlayer_3,WavePlayer_4,WavePlayer_5,WavePlayer_6,WavePlayer_7,WavePlayer_8]
  properties {weights [1,1,1,1,1,1,1]}
 }
 {
  id=Random_2
  children [WavePlayer_0, WavePlayer_9, WavePlayer_10, WavePlayer_11]
  properties {bShouldExcludeFromBranchCulling=true, weights [1,1,1,1]}
 }
 {id=WavePlayer_0, properties {soundWaveAssetPtr=&/Nug_Foley_Flap_010}}
 {id=WavePlayer_1, properties {soundWaveAssetPtr=&Nug_Foley_Flap_010}}
 {id=WavePlayer_10, properties {soundWaveAssetPtr=&/Nug_Foley_Flap_030}}
 {id=WavePlayer_11, properties {soundWaveAssetPtr=&/Nug_Foley_Flap_040}}
 {id=WavePlayer_3, properties {soundWaveAssetPtr=&Nug_Foley_Flap_030}}
 {id=WavePlayer_4, properties {soundWaveAssetPtr=&Nug_Foley_Flap_040}}
 {id=WavePlayer_5, properties {soundWaveAssetPtr=&/Nug_Foley_Flap_050}}
 {id=WavePlayer_6, properties {soundWaveAssetPtr=&/Nug_Foley_Flap_060}}
 {id=WavePlayer_7, properties {soundWaveAssetPtr=&/Nug_Foley_Flap_070}}
 {id=WavePlayer_8, properties {soundWaveAssetPtr=&/Nug_Foley_Flap_080}}
 {id=WavePlayer_9, properties {soundWaveAssetPtr=&/Nug_Foley_Flap_020}}
]
properties {
 attenuationSettings=/Irwin/Sounds/Settings/Prey_Step_Att
 bOverrideConcurrency=true
 maxDistance=3200
 soundClassObject=/Game/Sounds/FortSoundClassesAndModes/Fort_Creatures_Steps_Tactical
 volumeMultiplier=0.5
 concurrencyOverrides {maxCount=4}
}
