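# Sound cue for the Nug creature's egg-laying foley.
# It randomly picks one of two equally weighted AvacadoEater lay waves, then a modulator varies its pitch.
# It uses prey footstep attenuation and the tactical creature steps sound class.
# A concurrency override limits it to one instance per owner and blocks new plays while one is active.
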SoundCue /Irwin/Sounds/Nug/Foley/Nug_Foley_LayEgg_Cue
root=Modulator_2
nodes [
 {id=Modulator_2, children [Random_3], properties {pitchMax=1.2}}
 {id=Random_3, children [WavePlayer_3,WavePlayer_4], properties {weights [1,1]}}
 {id=WavePlayer_3, properties {soundWaveAssetPtr=&/AvacadoEater_Lay_02}}
 {id=WavePlayer_4, properties {soundWaveAssetPtr=&/AvacadoEater_Lay_01}}
]
properties {
 attenuationSettings=/Irwin/Sounds/Settings/Prey_Step_Att
 bOverrideConcurrency=true
 maxDistance=3200
 soundClassObject=/Game/Sounds/FortSoundClassesAndModes/Fort_Creatures_Steps_Tactical
 volumeMultiplier=1
 concurrencyOverrides {bLimitToOwner=true, maxCount=1, resolutionRule=PreventNew}
}
