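# Foley Sound Cue for a Nug spawn in Irwin, pairing a spawn sound with a slightly delayed startle flap.
# Each branch mixes a modulated random spawn wave with a short fixed-delay random flap.
# A quality node routes its first two slots to the full variation pools and the last two to smaller pools.
# Prey step attenuation applies, and owner-limited concurrency prevents overlapping plays.
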SoundCue /Irwin/Sounds/Nug/Foley/Nug_Foley_Spawn_Cue
root=QualityLevel_0
nodes [
 {id=Delay_0, children [Random_4], properties {delayMax=0.15, delayMin=0.15}}
 {id=Delay_1, children [Random_2], properties {delayMax=0.15, delayMin=0.15}}
 {id=Mixer_0, children [Modulator_1,Delay_0], properties {inputVolume [1,1]}}
 {id=Mixer_1, children [Modulator_0,Delay_1], properties {inputVolume [1,1]}}
 {id=Modulator_0, children [Random_0]}
 {id=Modulator_1, children [Random_1]}
 {id=QualityLevel_0, children [Mixer_0,Mixer_0,Mixer_1,Mixer_1]}
 {id=Random_0, children [WavePlayer_1,WavePlayer_2], properties {bShouldExcludeFromBranchCulling=true, weights [1,1]}}
 {id=Random_1, children [WavePlayer_0, WavePlayer_9, WavePlayer_10, WavePlayer_11], properties {weights [1,1,1,1]}}
 {id=Random_2, children [WavePlayer_5,WavePlayer_6,WavePlayer_7], properties {bShouldExcludeFromBranchCulling=true, weights [1,1,1]}}
 {id=Random_4, children [WavePlayer_19, WavePlayer_20, WavePlayer_21, WavePlayer_22, WavePlayer_23], properties {weights [1,1,1,1,1]}}
 {id=WavePlayer_0, properties {soundWaveAssetPtr=&/Nug_Foley_Spawn_010}}
 {id=WavePlayer_1, properties {soundWaveAssetPtr=&Nug_Foley_Spawn_010}}
 {id=WavePlayer_10, properties {soundWaveAssetPtr=&/Nug_Foley_Spawn_030}}
 {id=WavePlayer_11, properties {soundWaveAssetPtr=&/Nug_Foley_Spawn_040}}
 {id=WavePlayer_19, properties {soundWaveAssetPtr=&/Nug_Foley_StartleFlap_010}}
 {id=WavePlayer_2, properties {soundWaveAssetPtr=&/Nug_Foley_Spawn_020}}
 {id=WavePlayer_20, properties {soundWaveAssetPtr=&/Nug_Foley_StartleFlap_020}}
 {id=WavePlayer_21, properties {soundWaveAssetPtr=&/Nug_Foley_StartleFlap_030}}
 {id=WavePlayer_22, properties {soundWaveAssetPtr=&/Nug_Foley_StartleFlap_040}}
 {id=WavePlayer_23, properties {soundWaveAssetPtr=&/Nug_Foley_StartleFlap_050}}
 {id=WavePlayer_5, properties {soundWaveAssetPtr=&Nug_Foley_StartleFlap_010}}
 {id=WavePlayer_6, properties {soundWaveAssetPtr=&Nug_Foley_StartleFlap_020}}
 {id=WavePlayer_7, properties {soundWaveAssetPtr=&Nug_Foley_StartleFlap_030}}
 {id=WavePlayer_9, properties {soundWaveAssetPtr=&Nug_Foley_Spawn_020}}
]
properties {
 attenuationSettings=/Irwin/Sounds/Settings/Prey_Step_Att
 bOverrideConcurrency=true
 maxDistance=3200
 soundClassObject=/Game/Sounds/FortSoundClassesAndModes/Fort_Creatures_Steps_Tactical
 volumeMultiplier=1
 concurrencyOverrides {bLimitToOwner=true, maxCount=1, resolutionRule=PreventNew}
}
