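# Sound Cue for the Nug creature's landing foley, built from its footstep waves and played with prey gallop attenuation and the creature-steps sound class.
# A root Modulator feeds a Quality Level switch whose first two slots mix two random step pools, the second slightly delayed for a double impact.
# The last two slots use a single smaller random pool, excluded from branch culling.
# A concurrency override caps simultaneous instances, stopping the quietest.
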SoundCue /Irwin/Sounds/Nug/Foley/Nug_Land_Cue
root=Modulator_0
nodes [
 {id=Delay_0, children [Random_1], properties {delayMax=0.05, delayMin=0.01}}
 {id=Mixer_0, children [Random_0,Delay_0], properties {inputVolume [1,1]}}
 {id=Modulator_0, children [QualityLevel_0]}
 {id=QualityLevel_0, children [Mixer_0,Mixer_0,Random_2,Random_2]}
 {id=Random_0, children [WavePlayer_1,WavePlayer_2,WavePlayer_3,WavePlayer_4,WavePlayer_5], properties {weights [1,1,1,1,1]}}
 {id=Random_1, children [WavePlayer_6, WavePlayer_7, WavePlayer_8, WavePlayer_9, WavePlayer_10], properties {weights [1,1,1,1,1]}}
 {
  id=Random_2
  children [WavePlayer_0, WavePlayer_11, WavePlayer_12, WavePlayer_13]
  properties {bShouldExcludeFromBranchCulling=true, weights [1,1,1,1]}
 }
 {id=WavePlayer_0, properties {soundWaveAssetPtr=&/Nug_Step_09}}
 {id=WavePlayer_1, properties {soundWaveAssetPtr=&Nug_Step_09}}
 {id=WavePlayer_10, properties {soundWaveAssetPtr=&/Nug_Step_08}}
 {id=WavePlayer_11, properties {soundWaveAssetPtr=&/Nug_Step_10}}
 {id=WavePlayer_12, properties {soundWaveAssetPtr=&/Nug_Step_01}}
 {id=WavePlayer_13, properties {soundWaveAssetPtr=&/Nug_Step_02}}
 {id=WavePlayer_2, properties {soundWaveAssetPtr=&Nug_Step_10}}
 {id=WavePlayer_3, properties {soundWaveAssetPtr=&Nug_Step_01}}
 {id=WavePlayer_4, properties {soundWaveAssetPtr=&Nug_Step_02}}
 {id=WavePlayer_5, properties {soundWaveAssetPtr=&/Nug_Step_03}}
 {id=WavePlayer_6, properties {soundWaveAssetPtr=&/Nug_Step_04}}
 {id=WavePlayer_7, properties {soundWaveAssetPtr=&/Nug_Step_05}}
 {id=WavePlayer_8, properties {soundWaveAssetPtr=&/Nug_Step_06}}
 {id=WavePlayer_9, properties {soundWaveAssetPtr=&/Nug_Step_07}}
]
properties {
 attenuationSettings=/Irwin/Sounds/Settings/Prey_Gallop_Att
 bOverrideConcurrency=true
 maxDistance=3200
 soundClassObject=/Game/Sounds/FortSoundClassesAndModes/Fort_Creatures_Steps_Normal
 volumeMultiplier=1.5
 concurrencyOverrides {maxCount=4, resolutionRule=StopQuietest}
}
