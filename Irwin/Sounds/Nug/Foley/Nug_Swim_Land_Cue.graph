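# Sound cue for the Nug creature's swim landing foley, reusing Grandma swim-land recordings at doubled pitch and reduced volume.
# A quality-level switch selects a random pick from four takes or from a two-take subset, feeding a modulator with fixed volume.
# Prey step attenuation, a creature-steps sound class and a prey foley concurrency set govern playback.
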SoundCue /Irwin/Sounds/Nug/Foley/Nug_Swim_Land_Cue
root=Modulator_0
nodes [
 {id=Modulator_0, children [QualityLevel_0], properties {volumeMax=1, volumeMin=1}}
 {id=QualityLevel_0, children [Random_0,Random_0,Random_1,Random_1]}
 {id=Random_0, children [WavePlayer_2,WavePlayer_3,WavePlayer_4,WavePlayer_5], properties {weights [1,1,1,1]}}
 {id=Random_1, children [WavePlayer_0,WavePlayer_1], properties {bShouldExcludeFromBranchCulling=true, weights [1,1]}}
 {id=WavePlayer_0, properties {soundWaveAssetPtr=/Irwin/Sounds/Grandma/Foley/Foly_Grandma_Swim_Land_01}}
 {id=WavePlayer_1, properties {soundWaveAssetPtr=/Irwin/Sounds/Grandma/Foley/Foly_Grandma_Swim_Land_02}}
 {id=WavePlayer_2, properties {soundWaveAssetPtr=&Foly_Grandma_Swim_Land_01}}
 {id=WavePlayer_3, properties {soundWaveAssetPtr=&Foly_Grandma_Swim_Land_02}}
 {id=WavePlayer_4, properties {soundWaveAssetPtr=/Irwin/Sounds/Grandma/Foley/Foly_Grandma_Swim_Land_03}}
 {id=WavePlayer_5, properties {soundWaveAssetPtr=/Irwin/Sounds/Grandma/Foley/Foly_Grandma_Swim_Land_04}}
]
properties {
 attenuationSettings=/Irwin/Sounds/Settings/Prey_Step_Att
 maxDistance=3200
 pitchMultiplier=2
 soundClassObject=/Game/Sounds/FortSoundClassesAndModes/Fort_Creatures_Steps_Normal
 volumeMultiplier=0.7
 concurrencySet [/Irwin/Sounds/Settings/Conc/Prey_SetGroup_Foley_Conc]
}
